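# Physics asset for the Masked Warrior hair face accessory mesh, defining collision bodies and joint limits for its dynamic hair bones.
# Each body is a single capsule: five front strands, two back strands, a central hair mass and the head.
# All dyn_hair bones constrain directly to the head in a star layout rather than chains, with limited swing and twist.
# The strands share identical limits, while the hair mass has asymmetric swing.
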PhysicsAsset /BRCosmetics/Accessories/Hats/M_MED_Masked_Warrior_Hair_FaceAcc/Meshes/M_MED_Masked_Warrior_Hair_FaceAcc_Physics
bodies {
 dyn_hairBackA [{shapeType=Capsule, length=1.9648, radius=0.919, center=<1.2346 -0.2192 0.1899>, rotation=<rot 0 -100.0653 -98.6128>}]
 dyn_hairBackB [{shapeType=Capsule, length=1.8124, radius=0.8152, center=<0.8855 -0.1099 -0.1709>, rotation=<rot -1.3045 89.872 -97.9781>}]
 dyn_hairFrontA [{shapeType=Capsule, length=3.3571, radius=0.9628, center=<1.8356 -0.2392 0.1508>, rotation=<rot 0 -94.5062 -91.5393>}]
 dyn_hairFrontB [{shapeType=Capsule, length=5.8357, radius=1.1072, center=<2.092 -0.261 -0.076>, rotation=<rot 0 89.0237 -87.6309>}]
 dyn_hairFrontC [{shapeType=Capsule, length=3.2069, radius=1.1422, center=<1.7863 0.0366 -0.1235>, rotation=<rot 0 -98.1711 -87.2172>}]
 dyn_hairFrontD [{shapeType=Capsule, length=4.6749, radius=1.2831, center=<2.4191 -0.4551 0.5377>, rotation=<rot -5.0611 -92.2224 -98.989>}]
 dyn_hairFrontE [{shapeType=Capsule, length=3.5478, radius=1.0859, center=<1.1324 -0.1813 0.1728>, rotation=<rot 0 -91.4991 -87.9125>}]
 dyn_hairMass [{shapeType=Capsule, length=4.5837, radius=1.4163, center=<1.7885 0 0>, rotation=<rot 90 0 0>}]
 head [{shapeType=Capsule, length=2.4014, radius=1.8278}]
}
constraints [
 {bone1Name=dyn_hairMass, bone2Name=head, swing1LimitDegrees=12, swing2LimitDegrees=35, twistLimitDegrees=16}
 {bone1Name=dyn_hairBackA, bone2Name=head, swing1LimitDegrees=20, swing2LimitDegrees=20, twistLimitDegrees=16}
 {bone1Name=dyn_hairFrontD, bone2Name=head, swing1LimitDegrees=20, swing2LimitDegrees=20, twistLimitDegrees=16}
 {bone1Name=dyn_hairFrontE, bone2Name=head, swing1LimitDegrees=20, swing2LimitDegrees=20, twistLimitDegrees=16}
 {bone1Name=dyn_hairFrontC, bone2Name=head, swing1LimitDegrees=20, swing2LimitDegrees=20, twistLimitDegrees=16}
 {bone1Name=dyn_hairFrontB, bone2Name=head, swing1LimitDegrees=20, swing2LimitDegrees=20, twistLimitDegrees=16}
 {bone1Name=dyn_hairFrontA, bone2Name=head, swing1LimitDegrees=20, swing2LimitDegrees=20, twistLimitDegrees=16}
 {bone1Name=dyn_hairBackB, bone2Name=head, swing1LimitDegrees=20, swing2LimitDegrees=20, twistLimitDegrees=16}
]
